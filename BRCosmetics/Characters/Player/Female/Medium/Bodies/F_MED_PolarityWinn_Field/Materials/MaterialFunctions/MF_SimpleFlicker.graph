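# Material function that produces a simple flicker value between 0 and 1, stored with the PolarityWinn Field character body cosmetic.
# The Time input, defaulting to game time, is multiplied by Speed Scalar and drives three sine waves with slightly different periods.
# Each wave is remapped with ConstantBiasScale, then the three are averaged and saturated, so their drifting phases give an irregular-looking pulse without noise or textures.
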MaterialFunction /BRCosmetics/Characters/Player/Female/Medium/Bodies/F_MED_PolarityWinn_Field/Materials/MaterialFunctions/MF_SimpleFlicker
graph {nodes [
  n1 = FunctionInput() {bUsePreviewValueAsDefault=true, inputName=Speed Scalar, inputType=FunctionInput_Scalar, sortPriority=3, previewValue=<1 1 1 1>}
  n2 = Multiply(A=FunctionInput(Preview=Time()) {bUsePreviewValueAsDefault=true, inputName=Time, inputType=FunctionInput_Scalar}, B=@n1)
  n3 = Sine(None=@n2)
  n4 = Sine(None=@n2) {period=1.33}
  n5 = Add(A=Add(A=ConstantBiasScale(None=@n3), B=ConstantBiasScale(None=@n4)), B=ConstantBiasScale(None=Sine(None=@n2) {period=1.27}))
  FunctionOutput(None=Saturate(None=Divide(A=@n5) {constB=3}))
 ]}
