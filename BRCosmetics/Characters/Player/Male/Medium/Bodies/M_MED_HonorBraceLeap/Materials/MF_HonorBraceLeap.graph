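# Material function for the HonorBraceLeap male body cosmetic, layering a reactive emissive glow onto incoming material attributes.
# The Reactivity parameter drives a shifting level mask over texture-masked regions, including logo and speed-bar glows.
# Glow color comes from a curve-atlas gradient, combined with flickering fresnels, scrolling bi-planar plasma noise and mirrored pre-skinned sphere masks.
# It also adjusts base color, specular and roughness, passing metallic and normal through.
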MaterialFunction /BRCosmetics/Characters/Player/Male/Medium/Bodies/M_MED_HonorBraceLeap/Materials/MF_HonorBraceLeap
graph {nodes [
  FunctionInput() {bUsePreviewValueAsDefault=true, inputName=TechArtMask, sortPriority=3, previewValue=<1 1 1 1>}
  n1 = LocalPosition() {localOrigin=InstancePreSkinning}
  n2 = ShaderStageSwitch(PixelShader=VertexInterpolator(VS=@n1.XYZ).PS, VertexShader=@n1.XYZ)
  FunctionInput(Preview=@n2) {bUsePreviewValueAsDefault=true, inputName=PreSkinnedLocation, inputType=FunctionInput_Scalar, sortPriority=1, previewValue=<1 1 1 1>}
  n3 = GetMaterialAttributes(None=FunctionInput() {inputName=OUT_MA, inputType=FunctionInput_MaterialAttributes}) {attributeGetTypes [
    #1
    #2
    #3
    #4
    #5
    #6
    #7
   ]}
  n4 = SetMaterialAttributes!(
   MaterialAttributes=@n3.MaterialAttributes
   BaseColor=NamedRerouteUsage()
   Metallic=@n3.MaterialAttributes
   Specular=NamedRerouteUsage()
   Roughness=NamedRerouteUsage()
   Normal=@n3.MaterialAttributes
   EmissiveColor=NamedRerouteUsage()
  ) {attributeSetTypes [
    #1
    #2
    #3
    #4
    #5
    #6
   ]}
  FunctionOutput(None=@n4) {outputName=OUT_MA}
  LocalPosition() {localOrigin=InstancePreSkinning}
  n5 = Multiply(A=NamedRerouteUsage(), B=NamedRerouteUsage())
  NamedRerouteDeclaration(None=Multiply(A=NamedRerouteUsage(), B=OneMinus(None=@n5))) {name=OUT_BaseColor, nodeColor=<rgba 0 0.8856 1 1>}
  NamedRerouteDeclaration() {name=OUT_Metallic, nodeColor=<rgba 0.3462 1 0.0006 1>}
  NamedRerouteDeclaration(None=@n3.Normal) {name=IN_Normal, nodeColor=<rgba 0.25 0.25 1 1>}
  NamedRerouteDeclaration(None=@n3.EmissiveColor) {name=IN_EmissiveColor, nodeColor=<rgba 0 0.8 1 1>}
  n6 = TextureSampleParameter2D() {
   parameterName=Emissive
   samplerSource=SSM_Wrap_WorldGroupSettings
   texture=/BRCosmetics/Characters/Player/Male/Medium/Bodies/M_MED_HonorBraceLeap/Textures/T_HonorBraceLeap_Body_E
  }
  NamedRerouteDeclaration(None=Step(X=@n6.R) {constY=0.025}) {name=EmissiveMask, nodeColor=<rgba 0.25 1 0 1>}
  NamedRerouteDeclaration(None=@n2) {name=PreSkinnedLoc, nodeColor=<rgba 0.7001 0 1 1>}
  n7 = Subtract(A=NamedRerouteUsage(), B=ScalarParameter() {defaultValue=0.25, group=HonorBraceLeap, parameterName=FresnelBias}) {constB=0.25}
  n8 = MaterialFunctionCall!(Time, Speed Scalar=Constant() {r=4}) {materialFunction=/BRCosmetics/Characters/Player/Female/Medium/Bodies/F_MED_PolarityWinn_Field/Materials/MaterialFunctions/MF_SimpleFlicker}
  n9 = NamedRerouteDeclaration(None=@n8.Result) {name=Flicker, nodeColor=<rgba 1 0 0.2344 1>}
  n10 = NamedRerouteDeclaration(None=Fresnel(ExponentIn=LinearInterpolate(Alpha=@n9) {constA=0.9}) {baseReflectFraction=0, exponent=1}) {name=Fresnel, nodeColor=<rgba 0.9922 0 1 1>}
  n11 = Multiply(A=@n10, B=@n10)
  NamedRerouteDeclaration(None=Saturate(None=Add(A=@n7, B=@n11))) {name=SteppedFresnel, nodeColor=<rgba 1 0 0.6094 1>}
  n12 = Transform(None=Constant3Vector() {constant=<rgba 0 1 0 0>})
  n13 = LinearInterpolate(Alpha=@n9) {constA=0.5, constB=2}
  n14 = PixelNormalWS()
  NamedRerouteDeclaration(None=OneMinus(None=MaterialFunctionCall(Axis=@n12, Exponent=@n13, Normal=@n14) {materialFunction=&AxisAlignedFresnel}.Fresnel)) {name=AxisFresnel, nodeColor=<rgba 0.8516 1 0 1>}
  n15 = NamedRerouteUsage()
  n16 = MaterialFunctionCall(Speed Scalar=Constant() {r=1}) {materialFunction=&MF_SimpleFlicker}
  n17 = Saturate(None=Add(A=@n15, B=Multiply(A=@n16.Result) {constB=0.025}))
  n18 = LinearInterpolate(B=ScalarParameter() {defaultValue=8, group=HonorBraceLeap, parameterName=BottomGradientArmAdjust}, Alpha=@n15) {constA=6, constB=0.5}
  NamedRerouteDeclaration(None=Multiply(
    A=Power(Base=@n17, Exp=@n18) {constExponent=4}
    B=ScalarParameter() {defaultValue=10, group=HonorBraceLeap, parameterName=BottomGradientIntensity}
   ) {constB=10}) {name=BottomGradient, nodeColor=<rgba 0 0.7891 1 1>}
  n19 = Multiply(A=NamedRerouteUsage(), B=NamedRerouteUsage())
  NamedRerouteDeclaration(None=Multiply(A=NamedRerouteUsage(), B=OneMinus(None=@n19))) {name=OUT_Specular, nodeColor=<rgba 0.25 1 0 1>}
  n20 = LocalPosition() {localOrigin=InstancePreSkinning}
  n21 = Convert() {
   convertInputs [{
     type=Scalar
     defaultValue=<rgba 50 0 0 1>
     expressionInput {expression=null, inputName=None, mask=0, maskA=0, maskB=0, maskG=0, maskR=0, outputIndex=0}
    }]
   convertOutputs [{type=Vector3}]
  }
  n22 = Step(Y=ComponentMask(None=@n20.XYZ) {r=true}) {constX=0}
  n23 = SphereMask(A=@n20.XYZ, B=LinearInterpolate(A=@n21, B=Multiply(A=@n21) {constB=-1}, Alpha=@n22)) {attenuationRadius=80, hardnessPercent=0}
  n24 = Power(Base=ShaderStageSwitch(PixelShader=VertexInterpolator(VS=@n23).PS, VertexShader=@n23)) {constExponent=4}
  n25 = Convert() {
   convertInputs [{
     type=Scalar
     defaultValue=<rgba 25 0 0 1>
     expressionInput {expression=null, inputName=None, mask=0, maskA=0, maskB=0, maskG=0, maskR=0, outputIndex=0}
    }]
   convertOutputs [{type=Vector3}]
  }
  n26 = SphereMask(A=@n20.XYZ, B=LinearInterpolate(A=@n25, B=Multiply(A=@n25) {constB=-1}, Alpha=@n22)) {attenuationRadius=40, hardnessPercent=0}
  n27 = VertexInterpolator(VS=@n26)
  n28 = Multiply(A=Power(Base=ShaderStageSwitch(PixelShader=@n27.PS, VertexShader=@n26))) {constB=4}
  NamedRerouteDeclaration(None=LinearInterpolate(A=Multiply(A=@n24) {constB=8}, B=@n28, Alpha=@n28)) {name=LegVGradient, nodeColor=<rgba 1 0 0.6094 1>}
  NamedRerouteDeclaration(None=VertexInterpolator(VS=PreSkinnedNormal()).PS) {name=PreSkinnedNormal, nodeColor=<rgba 0.7001 0 1 1>}
  n29 = TextureSample() {
   samplerSource=SSM_Wrap_WorldGroupSettings
   samplerType=SAMPLERTYPE_LinearColor
   texture=/BRCosmetics/Characters/Player/Male/Medium/Bodies/M_MED_HonorBraceLeap/Textures/T_HonorBraceLeap_Body_Emissive_FX
  }
  NamedRerouteDeclaration(None=@n29.R) {name=ChargeOffset, nodeColor=<rgba 0.3359 0 1 1>}
  n30 = TextureObject() {texture=/Game/Global/Textures/Default/Noise/T_FireNoise_01}
  n31 = NamedRerouteUsage()
  n32 = Multiply(A=Time(), B=Constant3Vector() {constant=<rgba 0 0 -25 0>})
  n33 = Add(A=@n31, B=Divide(A=@n32))
  n34 = NamedRerouteUsage()
  n35 = Constant() {r=128}
  n36 = Power(Base=MaterialFunctionCall!(
    ColorTex=@n30
    NormalTex
    Pos=Add(A=@n31, B=@n32)
    Nrm=@n34
    ColorUVScale=Constant() {r=256}
    ColorBlendFactor
    NormalStrength
    NormalUVScale
    NormalBlendFactor
    Local Space Inputs
   ) {materialFunction=/Game/Global/MaterialFunctions/MF_BiPlanar}.Color)
  n37 = NamedRerouteUsage()
  n38 = Add(
   A=Power(Base=MaterialFunctionCall(ColorTex=@n30, Pos=@n33, Nrm=@n34, ColorUVScale=@n35) {materialFunction=&MF_BiPlanar}.Color)
   B=LinearInterpolate(Alpha=Multiply(A=@n36, B=@n37)) {constB=0.5}
  )
  NamedRerouteDeclaration(None=Saturate(None=ComponentMask(None=@n38) {r=true})) {name=Plasma, nodeColor=<rgba 0 0.8828 1 1>}
  n39 = FunctionInput() {bUsePreviewValueAsDefault=true, inputName=Fresnel, inputType=FunctionInput_Scalar, sortPriority=2, previewValue=<1 1 1 1>}
  NamedRerouteDeclaration(None=@n39) {name=IN_Fresnel, nodeColor=<rgba 0 1 0.6016 1>}
  n40 = ComponentMask(None=NamedRerouteUsage()) {b=true}
  NamedRerouteDeclaration(None=Step(X=@n40) {constX=0.55, constY=0.55}) {name=UpperBoddyMask, nodeColor=<rgba 1 0.7969 0 1>}
  n41 = Power(Base=NamedRerouteUsage(), Exp=Constant() {r=1.5})
  n42 = Constant() {r=8}
  n43 = Constant4Vector() {constant=<rgba 1 1 0 1>}
  n44 = Saturate(None=MaterialFunctionCall!(Mask In=OneMinus(None=@n40), Amount=@n43.R, Range=@n43.R, Start Offset=@n43.R, Sharpness=@n43.R) {materialFunction=/Game/Global/MaterialFunctions/MF_MaskShifting}.Result)
  n45 = Add(A=Multiply(A=NamedRerouteUsage(), B=Multiply(A=@n41, B=@n42) {constB=20}), B=@n44)
  NamedRerouteDeclaration(None=@n45) {name=FootGradient, nodeColor=<rgba 0 1 0.6016 1>}
  NamedRerouteDeclaration(None=@n44) {name=ReactiveGrow, nodeColor=<rgba 0.3828 1 0 1>}
  n46 = NamedRerouteUsage()
  n47 = Multiply(A=NamedRerouteUsage(), B=Multiply(A=Add(A=@n46) {constB=0.25}, B=NamedRerouteUsage()))
  NamedRerouteDeclaration(None=LinearInterpolate(A=NamedRerouteUsage(), Alpha=Saturate(None=@n47)) {constB=0.85}) {name=OUT_Roughness, nodeColor=<rgba 0.25 1 0 1>}
  NamedRerouteDeclaration(None=@n29.G) {name=GlowMask, nodeColor=<rgba 0 1 0.25 1>}
  n48 = TransformPosition(None=NamedRerouteUsage()) {transformType=TRANSFORMPOSSOURCE_Instance}
  n49 = PreSkinnedLocalBounds()
  NamedRerouteDeclaration(None=Divide(A=Subtract(A=@n48, B=@n49.Min), B=@n49.Extents)) {name=Bounding, nodeColor=<rgba 0 0.5312 1 1>}
  NamedRerouteDeclaration(None=@n3.AmbientOcclusion) {name=AO, nodeColor=<rgba 1 0 0.0703 1>}
  NamedRerouteDeclaration() {name=OUT_Normal, nodeColor=<rgba 0.25 0.25 1 1>}
  n50 = CurveAtlasRowParameter(CurveTime=Multiply(A=@n45, B=NamedRerouteUsage())) {
   atlas=/BRCosmetics/Characters/Player/Male/Medium/Bodies/M_MED_HonorBraceLeap/Textures/Curves/CA_HornorBranceLeap
   curve=/BRCosmetics/Characters/Player/Male/Medium/Bodies/M_MED_HonorBraceLeap/Textures/Curves/C_HornorBraceLeap_Gradient
   group=HonorBraceLeap
   parameterName=EmissiveGradient
  }
  n51 = Multiply(A=NamedRerouteUsage(), B=@n50)
  n52 = NamedRerouteUsage()
  n53 = LinearInterpolate(A=NamedRerouteUsage(), B=@n51, Alpha=@n52)
  n54 = Multiply(A=Multiply(A=@n51, B=Power(Base=@n52) {constExponent=8})) {constB=0.5}
  n55 = Add(A=@n51, B=Multiply(A=Add(A=@n53, B=@n54)) {constB=2})
  n56 = LocalPosition() {localOrigin=InstancePreSkinning}
  n57 = Convert() {
   convertInputs [{
     type=Scalar
     defaultValue=<rgba 20 0 0 1>
     expressionInput {expression=null, inputName=None, mask=0, maskA=0, maskB=0, maskG=0, maskR=0, outputIndex=0}
    }]
   convertOutputs [{type=Vector3, defaultValue=<rgba 0 0 145 1>}, {type=Vector3, defaultValue=<rgba -1 1 1 1>}]
  }
  n58 = Multiply(A=@n57, B=@n57) {constB=-1}
  n59 = Step(Y=ComponentMask(None=@n56.XYZ) {r=true}) {constX=0}
  n60 = Convert() {
   convertInputs [{
     type=Scalar
     defaultValue=<rgba 10 0 0 1>
     expressionInput {expression=null, inputName=None, mask=0, maskA=0, maskB=0, maskG=0, maskR=0, outputIndex=0}
    }]
   convertOutputs [{type=Vector3, defaultValue=<rgba 0 3 102 1>}, {type=Vector3, defaultValue=<rgba -1 1 1 1>}]
  }
  n61 = Multiply(A=@n60, B=@n60) {constB=-1}
  n62 = Step(Y=ComponentMask(None=@n56.XYZ) {r=true}) {constX=0}
  n63 = Add(
   A=SphereMask(A=@n56.XYZ, B=LinearInterpolate(A=@n57, B=@n58, Alpha=@n59)) {attenuationRadius=18, hardnessPercent=0}
   B=SphereMask(A=@n56.XYZ, B=LinearInterpolate(A=@n60, B=@n61, Alpha=@n62)) {attenuationRadius=5, hardnessPercent=0}
  )
  n64 = VertexInterpolator(VS=@n63)
  n65 = Multiply(A=NamedRerouteUsage(), B=Step(Y=ShaderStageSwitch(PixelShader=@n64.PS, VertexShader=@n63)) {constX=0.1})
  n66 = Constant4Vector() {constant=<rgba 0 0.85 0 1>}
  n67 = MaterialFunctionCall(
   Input=ScalarParameter() {group=HonorBraceLeap, parameterName=Reactivity}
   Input Low=@n66.R
   Input High=@n66.R
   Target Low=@n66.R
   Target High=@n66.R
  ) {materialFunction=&RemapValueRange}
  n68 = OneMinus(None=@n67.Result)
  n69 = OneMinus(None=NamedRerouteUsage())
  n70 = Subtract(A=@n68, B=@n69) {constB=0.5}
  n71 = Constant4Vector() {constant=<rgba 1 0.5 0 1>}
  n72 = SmoothStep(
   Min=@n70
   Max=Add(A=@n70, B=ScalarParameter() {group=HonorBraceLeap, parameterName=Leveling_Smoothness}) {constB=0.5}
   Value=Saturate(None=MaterialFunctionCall(Mask In=@n69, Amount=@n68, Range=@n71.G, Start Offset=@n71.G, Sharpness=@n71.G) {materialFunction=&MF_MaskShifting}.Result)
  )
  n73 = NamedRerouteUsage()
  n74 = NamedRerouteUsage()
  n75 = Multiply(A=@n67.Result, B=@n74)
  n76 = NamedRerouteUsage()
  n77 = Step(Y=Constant() {r=0.45}, X=OneMinus(None=ComponentMask(None=@n76) {b=true}))
  n78 = Multiply(A=Power(Base=Subtract(A=@n74, B=@n73)), B=Multiply(A=@n72, B=@n77)) {desc=SpeedBarGlow}
  n79 = Multiply(
   A=Multiply(A=Multiply(A=@n55, B=@n65), B=ScalarParameter() {defaultValue=1, group=HonorBraceLeap, parameterName=ReactivityFX_Intensity})
   B=Add(A=Multiply(A=@n72, B=@n73), B=LinearInterpolate(A=@n75, B=@n78, Alpha=@n77) {desc=LogoGlow})
  )
  n80 = ScalarParameter() {defaultValue=1, group=HonorBraceLeap, parameterName=Reactivity_Toggle}
  NamedRerouteDeclaration(None=Add(A=NamedRerouteUsage(), B=Multiply(A=@n79, B=@n80))) {name=OUT_EmissiveColor, nodeColor=<rgba 0 0.8 1 1>}
  NamedRerouteDeclaration(None=@n3.BaseColor) {name=IN_BaseColor, nodeColor=<rgba 0 0.8 1 1>}
  NamedRerouteDeclaration(None=@n3.Metallic) {name=IN_Metallic, nodeColor=<rgba 0.25 1 0 1>}
  NamedRerouteDeclaration(None=@n3.Specular) {name=IN_Specular, nodeColor=<rgba 0.25 1 0 1>}
  NamedRerouteDeclaration(None=@n3.Roughness) {name=IN_Roughness, nodeColor=<rgba 0.25 1 0 1>}
  NamedRerouteUsage()
  NamedRerouteUsage()
 ]}
