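# Post-copy AnimBlueprint for the M_MED_HonorBraceLeap male body cosmetic, following the attached parent mesh's pose and adjusting its accessory bones.
# Copy Pose From Mesh feeds chained multi-target weighted constraints that place both shoulder pads from arm, clavicle and spine bones, and both wallet bases from pelvis and spine.
# The result passes through the body's Control Rig, driven by crouch, emote, front-end, parachute and skydive flags.
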AnimBlueprint /BRCosmetics/Characters/Player/Male/Medium/Bodies/M_MED_HonorBraceLeap/Meshes/M_MED_HonorBraceLeap_AnimBP
target_skeleton=&/M_MED_HonorBraceLeap_Skeleton
graphs [
 AnimGraph {nodes [
   Control Rig() {
    alphaBoolBlend {blendOption=Linear}
    controlRigAssetReference {blueprintRigClass=&/M_MED_HonorBraceLeap_CtrlRig.M_MED_HonorBraceLeap_CtrlRig_C}
   }
   n1 = "Constraint - base_l_shoulderpad (PPPPPPPP)"!(
    ComponentPose=Local To Component(LocalPose=Copy Pose From Mesh() {bUseAttachedParent=true})
    Alpha=1
    ConstraintWeights_0=0.377
    ConstraintWeights_1:float=0.2473
    ConstraintWeights_2:float=0.1614
    ConstraintWeights_3:float=0.1115
    ConstraintWeights_4:float=0.035
    ConstraintWeights_5:float=0.0299
    ConstraintWeights_6:float=0.0197
    ConstraintWeights_7:float=0.0183
   ) -> (Pose) {
    alphaBoolBlend {blendOption=Linear}
    boneToModify {boneName=base_l_shoulderpad}
    constraintSetup [
     {targetBone {boneName=upperarm_twist_01_l}}
     {targetBone {boneName=deform_upperarm_up_l}}
     {targetBone {boneName=clavicle_l}}
     {targetBone {boneName=upperarm_twist_02_l}}
     {targetBone {boneName=spine_05}}
     {targetBone {boneName=deform_upperarm_fr_l}}
     {targetBone {boneName=upperarm_l}}
     {targetBone {boneName=spine_03}}
    ]
    constraintWeights [1,1,1,1,1,1,1,1]
   }
   n2 = "Constraint - base_r_shoulderpad (PPPPPPPP)"!(
    ComponentPose=@n1
    Alpha=1
    ConstraintWeights_0=0.2863
    ConstraintWeights_1:float=0.3221
    ConstraintWeights_2:float=0.1792
    ConstraintWeights_3:float=0.0813
    ConstraintWeights_4:float=0.0365
    ConstraintWeights_5:float=0.0689
    ConstraintWeights_6:float=0.0037
    ConstraintWeights_7:float=0.022
   ) -> (Pose) {
    alphaBoolBlend {blendOption=Linear}
    boneToModify {boneName=base_r_shoulderpad}
    constraintSetup [
     {targetBone {boneName=upperarm_twist_01_r}}
     {targetBone {boneName=deform_upperarm_up_r}}
     {targetBone {boneName=clavicle_r}}
     {targetBone {boneName=upperarm_twist_02_r}}
     {targetBone {boneName=spine_05}}
     {targetBone {boneName=deform_upperarm_fr_r}}
     {targetBone {boneName=upperarm_r}}
     {targetBone {boneName=spine_03}}
    ]
    constraintWeights [1,1,1,1,1,1,1,1]
   }
   n3 = "Constraint - wallet_basa_r (PPPPP)"!(
    ComponentPose=@n2
    Alpha=1
    ConstraintWeights_0=0.703
    ConstraintWeights_1:float=0.124
    ConstraintWeights_2:float=0.096
    ConstraintWeights_3:float=0.073
    ConstraintWeights_4:float=0.004
   ) -> (Pose) {
    alphaBoolBlend {blendOption=Linear}
    boneToModify {boneName=wallet_basa_r}
    constraintSetup [
     {targetBone {boneName=pelvis}}
     {targetBone {boneName=spine_01}}
     {targetBone {boneName=spine_02}}
     {targetBone {boneName=spine_03}}
     {targetBone {boneName=spine_04}}
    ]
    constraintWeights [1,1,1,1,1]
   }
   n4 = "Constraint - wallet_basa_l (PPPPP)"!(
    ComponentPose=@n3
    Alpha=1
    ConstraintWeights_0=0.704
    ConstraintWeights_1:float=0.097
    ConstraintWeights_2:float=0.123
    ConstraintWeights_3:float=0.073
    ConstraintWeights_4:float=0.003
   ) -> (Pose) {
    alphaBoolBlend {blendOption=Linear}
    boneToModify {boneName=wallet_basa_l}
    constraintSetup [
     {targetBone {boneName=pelvis}}
     {targetBone {boneName=spine_01}}
     {targetBone {boneName=spine_02}}
     {targetBone {boneName=spine_03}}
     {targetBone {boneName=spine_04}}
    ]
    constraintWeights [1,1,1,1,1]
   }
   n5 = Control Rig!(
    Source=Component To Local(ComponentPose=@n4)
    Alpha
    Crouching:bool=Get bIsCrouching()
    PlayingEmote:bool=Get bIsPlayingEmote()
    FrontEnd:bool=Get bIsInFrontEnd()
    Parachute Opened:bool=Get bIsParachuteOpened()
    Skydive:bool=Get bIsSkyDiving()
   ) -> (Pose) {
    tag=cr
    alphaBoolBlend {blendOption=Linear}
    controlRigAssetReference {blueprintRigClass=&M_MED_HonorBraceLeap_CtrlRig_C}
   }
   Output Pose(Result=@n5) {layerGroup=None}
  ]}
 EventGraph {nodes [Event BlueprintUpdateAnimation()]}
]
variables []
