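# Anim Blueprint for the HonorBraceLeap male character's FaceAcc part, adding secondary physics motion to its hair bones.
# It copies the attached parent mesh's pose, then chains 36 single-bone Anim Dynamics nodes simulated relative to the head.
# Angular springs and three tiers of limits and damping range from loose to stiff.
# The update event flips buckle_Gravity Override upward during skydive floating, but no node here reads it.
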AnimBlueprint /BRCosmetics/Characters/Player/Male/Medium/Bodies/M_MED_HonorBraceLeap/Meshes/Parts/M_MED_HonorBraceLeap_FaceAcc_AnimBP
target_skeleton=&/M_MED_HonorBraceLeap_FaceAcc_Skeleton
graphs [
 AnimGraph {nodes [
   n1 = Anim Dynamics - Bone: dyn_hair_a!(
    ComponentPose=Local To Component(LocalPose=Copy Pose From Mesh() {bUseAttachedParent=true})
    Alpha=1
    GravityOverride:Vector="0,0,-980"
    ExternalForce="0,0,0"
   ) -> (Pose) {
    angularBiasOverride=0.2
    angularDampingOverride=0.99
    angularSpringConstant=14000
    bAngularSpring=true
    bOverrideAngularBias=true
    bOverrideAngularDamping=true
    bUseGravityOverride=true
    gravityScale=0.35
    alphaBoolBlend {blendOption=Linear}
    boundBone {boneName=dyn_hair_a}
    componentAppliedLinearAccClamp=<200000 200000 200000>
    componentLinearAccScale=<1 1 1>
    componentLinearVelScale=<2 2 2>
    gravityOverride {z=-980}
    physicsBodyDefinitions [{
      boundBone {boneName=dyn_hair_a}
      boxExtents=<40 50 50>
      constraintSetup {angularLimitsMax=<15 80 15>, angularLimitsMin=<-15 -10 -15>, angularTarget=<1 0 0>}
      localJointOffset=<1.8 0 0>
     }]
    relativeSpaceBone {boneName=head}
   }
   n2 = Anim Dynamics - Bone: dyn_hair_y!(ComponentPose=@n1, Alpha=1, GravityOverride:Vector="0,0,-980", ExternalForce="0,0,0") -> (Pose) {
    angularBiasOverride=0.2
    angularDampingOverride=0.99
    angularSpringConstant=14000
    bAngularSpring=true
    bOverrideAngularBias=true
    bOverrideAngularDamping=true
    bUseGravityOverride=true
    gravityScale=0.35
    alphaBoolBlend {blendOption=Linear}
    boundBone {boneName=dyn_hair_y}
    componentAppliedLinearAccClamp=<200000 200000 200000>
    componentLinearAccScale=<1 1 1>
    componentLinearVelScale=<2 2 2>
    gravityOverride {z=-980}
    physicsBodyDefinitions [{
      boundBone {boneName=dyn_hair_y}
      boxExtents=<40 50 50>
      constraintSetup {angularLimitsMax=<15 80 15>, angularLimitsMin=<-15 -10 -15>, angularTarget=<1 0 0>}
      localJointOffset=<1.8 0 0>
     }]
    relativeSpaceBone {boneName=head}
   }
   n3 = Anim Dynamics - Bone: dyn_hair_v!(ComponentPose=@n2, Alpha=1, GravityOverride:Vector="0,0,-980", ExternalForce="0,0,0") -> (Pose) {
    angularBiasOverride=0.2
    angularDampingOverride=0.99
    angularSpringConstant=14000
    bAngularSpring=true
    bOverrideAngularBias=true
    bOverrideAngularDamping=true
    bUseGravityOverride=true
    gravityScale=0.35
    alphaBoolBlend {blendOption=Linear}
    boundBone {boneName=dyn_hair_v}
    componentAppliedLinearAccClamp=<200000 200000 200000>
    componentLinearAccScale=<1 1 1>
    componentLinearVelScale=<2 2 2>
    gravityOverride {z=-980}
    physicsBodyDefinitions [{
      boundBone {boneName=dyn_hair_v}
      boxExtents=<40 50 50>
      constraintSetup {angularLimitsMax=<15 80 15>, angularLimitsMin=<-15 -10 -15>, angularTarget=<1 0 0>}
      localJointOffset=<1.8 0 0>
     }]
    relativeSpaceBone {boneName=head}
   }
   n4 = Anim Dynamics - Bone: dyn_hair_c!(ComponentPose=@n3, Alpha=1, GravityOverride:Vector="0,0,-980", ExternalForce="0,0,0") -> (Pose) {
    angularBiasOverride=0.2
    angularDampingOverride=0.99
    angularSpringConstant=14000
    bAngularSpring=true
    bOverrideAngularBias=true
    bOverrideAngularDamping=true
    bUseGravityOverride=true
    gravityScale=0.35
    alphaBoolBlend {blendOption=Linear}
    boundBone {boneName=dyn_hair_c}
    componentAppliedLinearAccClamp=<200000 200000 200000>
    componentLinearAccScale=<1 1 1>
    componentLinearVelScale=<2 2 2>
    gravityOverride {z=-980}
    physicsBodyDefinitions [{
      boundBone {boneName=dyn_hair_c}
      boxExtents=<40 50 50>
      constraintSetup {angularLimitsMax=<15 80 15>, angularLimitsMin=<-15 -10 -15>, angularTarget=<1 0 0>}
      localJointOffset=<1.8 0 0>
     }]
    relativeSpaceBone {boneName=head}
   }
   n5 = Anim Dynamics - Bone: dyn_hair_w!(ComponentPose=@n4, Alpha=1, GravityOverride:Vector="0,0,-980", ExternalForce="0,0,0") -> (Pose) {
    angularBiasOverride=0.2
    angularDampingOverride=0.99
    angularSpringConstant=14000
    bAngularSpring=true
    bOverrideAngularBias=true
    bOverrideAngularDamping=true
    bUseGravityOverride=true
    gravityScale=0.35
    alphaBoolBlend {blendOption=Linear}
    boundBone {boneName=dyn_hair_w}
    componentAppliedLinearAccClamp=<200000 200000 200000>
    componentLinearAccScale=<1 1 1>
    componentLinearVelScale=<2 2 2>
    gravityOverride {z=-980}
    physicsBodyDefinitions [{
      boundBone {boneName=dyn_hair_w}
      boxExtents=<40 50 50>
      constraintSetup {angularLimitsMax=<15 80 15>, angularLimitsMin=<-15 -10 -15>, angularTarget=<1 0 0>}
      localJointOffset=<1.8 0 0>
     }]
    relativeSpaceBone {boneName=head}
   }
   n6 = Anim Dynamics - Bone: dyn_hair_g!(ComponentPose=@n5, Alpha=1, GravityOverride:Vector="0,0,-980", ExternalForce="0,0,0") -> (Pose) {
    angularBiasOverride=0.2
    angularDampingOverride=0.99
    angularSpringConstant=14000
    bAngularSpring=true
    bOverrideAngularBias=true
    bOverrideAngularDamping=true
    bUseGravityOverride=true
    gravityScale=0.35
    alphaBoolBlend {blendOption=Linear}
    boundBone {boneName=dyn_hair_g}
    componentAppliedLinearAccClamp=<200000 200000 200000>
    componentLinearAccScale=<1 1 1>
    componentLinearVelScale=<2 2 2>
    gravityOverride {z=-980}
    physicsBodyDefinitions [{
      boundBone {boneName=dyn_hair_g}
      boxExtents=<40 50 50>
      constraintSetup {angularLimitsMax=<15 80 15>, angularLimitsMin=<-15 -10 -15>, angularTarget=<1 0 0>}
      localJointOffset=<1.8 0 0>
     }]
    relativeSpaceBone {boneName=head}
   }
   n7 = Anim Dynamics - Bone: dyn_hair_q!(ComponentPose=@n6, Alpha=1, GravityOverride:Vector="0,0,-980", ExternalForce="0,0,0") -> (Pose) {
    angularBiasOverride=0.2
    angularDampingOverride=0.99
    angularSpringConstant=14000
    bAngularSpring=true
    bOverrideAngularBias=true
    bOverrideAngularDamping=true
    bUseGravityOverride=true
    gravityScale=0.35
    alphaBoolBlend {blendOption=Linear}
    boundBone {boneName=dyn_hair_q}
    componentAppliedLinearAccClamp=<200000 200000 200000>
    componentLinearAccScale=<1 1 1>
    componentLinearVelScale=<2 2 2>
    gravityOverride {z=-980}
    physicsBodyDefinitions [{
      boundBone {boneName=dyn_hair_q}
      boxExtents=<40 50 50>
      constraintSetup {angularLimitsMax=<15 80 15>, angularLimitsMin=<-15 -10 -15>, angularTarget=<1 0 0>}
      localJointOffset=<1.8 0 0>
     }]
    relativeSpaceBone {boneName=head}
   }
   n8 = Anim Dynamics - Bone: dyn_hair_h!(ComponentPose=@n7, Alpha=1, GravityOverride:Vector="0,0,-980", ExternalForce="0,0,0") -> (Pose) {
    angularBiasOverride=0.2
    angularDampingOverride=0.99
    angularSpringConstant=14000
    bAngularSpring=true
    bOverrideAngularBias=true
    bOverrideAngularDamping=true
    bUseGravityOverride=true
    gravityScale=0.35
    alphaBoolBlend {blendOption=Linear}
    boundBone {boneName=dyn_hair_h}
    componentAppliedLinearAccClamp=<200000 200000 200000>
    componentLinearAccScale=<1 1 1>
    componentLinearVelScale=<2 2 2>
    gravityOverride {z=-980}
    physicsBodyDefinitions [{
      boundBone {boneName=dyn_hair_h}
      boxExtents=<40 50 50>
      constraintSetup {angularLimitsMax=<15 80 15>, angularLimitsMin=<-15 -10 -15>, angularTarget=<1 0 0>}
      localJointOffset=<1.8 0 0>
     }]
    relativeSpaceBone {boneName=head}
   }
   n9 = Anim Dynamics - Bone: dyn_hair_ss!(ComponentPose=@n8, Alpha=1, GravityOverride:Vector="0,0,-980", ExternalForce="0,0,0") -> (Pose) {
    angularBiasOverride=0.2
    angularDampingOverride=0.99
    angularSpringConstant=14000
    bAngularSpring=true
    bOverrideAngularBias=true
    bOverrideAngularDamping=true
    bUseGravityOverride=true
    gravityScale=0.35
    alphaBoolBlend {blendOption=Linear}
    boundBone {boneName=dyn_hair_ss}
    componentAppliedLinearAccClamp=<200000 200000 200000>
    componentLinearAccScale=<1 1 1>
    componentLinearVelScale=<2 2 2>
    gravityOverride {z=-980}
    physicsBodyDefinitions [{
      boundBone {boneName=dyn_hair_ss}
      boxExtents=<40 50 50>
      constraintSetup {angularLimitsMax=<15 80 15>, angularLimitsMin=<-15 -10 -15>, angularTarget=<1 0 0>}
      localJointOffset=<1.8 0 0>
     }]
    relativeSpaceBone {boneName=head}
   }
   n10 = Anim Dynamics - Bone: dyn_hair_cc!(ComponentPose=@n9, Alpha=1, GravityOverride:Vector="0,0,-980", ExternalForce="0,0,0") -> (Pose) {
    angularBiasOverride=0.2
    angularDampingOverride=0.99
    angularSpringConstant=14000
    bAngularSpring=true
    bOverrideAngularBias=true
    bOverrideAngularDamping=true
    bUseGravityOverride=true
    gravityScale=0.35
    alphaBoolBlend {blendOption=Linear}
    boundBone {boneName=dyn_hair_cc}
    componentAppliedLinearAccClamp=<200000 200000 200000>
    componentLinearAccScale=<1 1 1>
    componentLinearVelScale=<2 2 2>
    gravityOverride {z=-980}
    physicsBodyDefinitions [{
      boundBone {boneName=dyn_hair_cc}
      boxExtents=<40 50 50>
      constraintSetup {angularLimitsMax=<15 80 15>, angularLimitsMin=<-15 -10 -15>, angularTarget=<1 0 0>}
      localJointOffset=<1.8 0 0>
     }]
    relativeSpaceBone {boneName=head}
   }
   n11 = Anim Dynamics - Bone: dyn_hair_r!(ComponentPose=@n10, Alpha=1, GravityOverride:Vector="0,0,-980", ExternalForce="0,0,0") -> (Pose) {
    angularBiasOverride=0.2
    angularDampingOverride=0.99
    angularSpringConstant=14000
    bAngularSpring=true
    bOverrideAngularBias=true
    bOverrideAngularDamping=true
    bUseGravityOverride=true
    gravityScale=0.35
    alphaBoolBlend {blendOption=Linear}
    boundBone {boneName=dyn_hair_r}
    componentAppliedLinearAccClamp=<200000 200000 200000>
    componentLinearAccScale=<1 1 1>
    componentLinearVelScale=<2 2 2>
    gravityOverride {z=-980}
    physicsBodyDefinitions [{
      boundBone {boneName=dyn_hair_r}
      boxExtents=<40 50 50>
      constraintSetup {angularLimitsMax=<15 80 15>, angularLimitsMin=<-15 -10 -15>, angularTarget=<1 0 0>}
      localJointOffset=<1.8 0 0>
     }]
    relativeSpaceBone {boneName=head}
   }
   n12 = Anim Dynamics - Bone: dyn_hair_b!(ComponentPose=@n11, Alpha=1, GravityOverride:Vector="0,0,-980", ExternalForce="0,0,0") -> (Pose) {
    angularBiasOverride=0.2
    angularDampingOverride=0.999
    angularSpringConstant=14000
    bAngularSpring=true
    bOverrideAngularBias=true
    bOverrideAngularDamping=true
    bUseGravityOverride=true
    gravityScale=0.35
    alphaBoolBlend {blendOption=Linear}
    boundBone {boneName=dyn_hair_b}
    componentAppliedLinearAccClamp=<200000 200000 200000>
    componentLinearAccScale=<1 1 1>
    componentLinearVelScale=<1 1 1>
    gravityOverride {z=-980}
    physicsBodyDefinitions [{
      boundBone {boneName=dyn_hair_b}
      boxExtents=<30 30 30>
      constraintSetup {angularLimitsMax=<7 20 7>, angularLimitsMin=<-7 -7.5 -7>, angularTarget=<1 0 0>}
      localJointOffset=<1.4 0 0>
     }]
    relativeSpaceBone {boneName=head}
   }
   n13 = Anim Dynamics - Bone: dyn_hair_d!(ComponentPose=@n12, Alpha=1, GravityOverride:Vector="0,0,-980", ExternalForce="0,0,0") -> (Pose) {
    angularBiasOverride=0.2
    angularDampingOverride=0.999
    angularSpringConstant=14000
    bAngularSpring=true
    bOverrideAngularBias=true
    bOverrideAngularDamping=true
    bUseGravityOverride=true
    gravityScale=0.35
    alphaBoolBlend {blendOption=Linear}
    boundBone {boneName=dyn_hair_d}
    componentAppliedLinearAccClamp=<200000 200000 200000>
    componentLinearAccScale=<1 1 1>
    componentLinearVelScale=<1 1 1>
    gravityOverride {z=-980}
    physicsBodyDefinitions [{
      boundBone {boneName=dyn_hair_d}
      boxExtents=<30 30 30>
      constraintSetup {angularLimitsMax=<7 20 7>, angularLimitsMin=<-7 -7.5 -7>, angularTarget=<1 0 0>}
      localJointOffset=<1.4 0 0>
     }]
    relativeSpaceBone {boneName=head}
   }
   n14 = Anim Dynamics - Bone: dyn_hair_n!(ComponentPose=@n13, Alpha=1, GravityOverride:Vector="0,0,-980", ExternalForce="0,0,0") -> (Pose) {
    angularBiasOverride=0.2
    angularDampingOverride=0.999
    angularSpringConstant=14000
    bAngularSpring=true
    bOverrideAngularBias=true
    bOverrideAngularDamping=true
    bUseGravityOverride=true
    gravityScale=0.35
    alphaBoolBlend {blendOption=Linear}
    boundBone {boneName=dyn_hair_n}
    componentAppliedLinearAccClamp=<200000 200000 200000>
    componentLinearAccScale=<1 1 1>
    componentLinearVelScale=<1 1 1>
    gravityOverride {z=-980}
    physicsBodyDefinitions [{
      boundBone {boneName=dyn_hair_n}
      boxExtents=<30 30 30>
      constraintSetup {angularLimitsMax=<7 20 7>, angularLimitsMin=<-7 -7.5 -7>, angularTarget=<1 0 0>}
      localJointOffset=<1.4 0 0>
     }]
    relativeSpaceBone {boneName=head}
   }
   n15 = Anim Dynamics - Bone: dyn_hair_f!(ComponentPose=@n14, Alpha=1, GravityOverride:Vector="0,0,-980", ExternalForce="0,0,0") -> (Pose) {
    angularBiasOverride=0.2
    angularDampingOverride=0.999
    angularSpringConstant=14000
    bAngularSpring=true
    bOverrideAngularBias=true
    bOverrideAngularDamping=true
    bUseGravityOverride=true
    gravityScale=0.35
    alphaBoolBlend {blendOption=Linear}
    boundBone {boneName=dyn_hair_f}
    componentAppliedLinearAccClamp=<200000 200000 200000>
    componentLinearAccScale=<1 1 1>
    componentLinearVelScale=<1 1 1>
    gravityOverride {z=-980}
    physicsBodyDefinitions [{
      boundBone {boneName=dyn_hair_f}
      boxExtents=<30 30 30>
      constraintSetup {angularLimitsMax=<7 20 7>, angularLimitsMin=<-7 -7.5 -7>, angularTarget=<1 0 0>}
      localJointOffset=<1.4 0 0>
     }]
    relativeSpaceBone {boneName=head}
   }
   n16 = Anim Dynamics - Bone: dyn_hair_p!(ComponentPose=@n15, Alpha=1, GravityOverride:Vector="0,0,-980", ExternalForce="0,0,0") -> (Pose) {
    angularBiasOverride=0.2
    angularDampingOverride=0.999
    angularSpringConstant=14000
    bAngularSpring=true
    bOverrideAngularBias=true
    bOverrideAngularDamping=true
    bUseGravityOverride=true
    gravityScale=0.35
    alphaBoolBlend {blendOption=Linear}
    boundBone {boneName=dyn_hair_p}
    componentAppliedLinearAccClamp=<200000 200000 200000>
    componentLinearAccScale=<1 1 1>
    componentLinearVelScale=<1 1 1>
    gravityOverride {z=-980}
    physicsBodyDefinitions [{
      boundBone {boneName=dyn_hair_p}
      boxExtents=<30 30 30>
      constraintSetup {angularLimitsMax=<7 20 7>, angularLimitsMin=<-7 -7.5 -7>, angularTarget=<1 0 0>}
      localJointOffset=<1.4 0 0>
     }]
    relativeSpaceBone {boneName=head}
   }
   n17 = Anim Dynamics - Bone: dyn_hair_s!(ComponentPose=@n16, Alpha=1, GravityOverride:Vector="0,0,-980", ExternalForce="0,0,0") -> (Pose) {
    angularBiasOverride=0.2
    angularDampingOverride=0.999
    angularSpringConstant=14000
    bAngularSpring=true
    bOverrideAngularBias=true
    bOverrideAngularDamping=true
    bUseGravityOverride=true
    gravityScale=0.35
    alphaBoolBlend {blendOption=Linear}
    boundBone {boneName=dyn_hair_s}
    componentAppliedLinearAccClamp=<200000 200000 200000>
    componentLinearAccScale=<1 1 1>
    componentLinearVelScale=<1 1 1>
    gravityOverride {z=-980}
    physicsBodyDefinitions [{
      boundBone {boneName=dyn_hair_s}
      boxExtents=<30 30 30>
      constraintSetup {angularLimitsMax=<7 20 7>, angularLimitsMin=<-7 -7.5 -7>, angularTarget=<1 0 0>}
      localJointOffset=<1.4 0 0>
     }]
    relativeSpaceBone {boneName=head}
   }
   n18 = Anim Dynamics - Bone: dyn_hair_t!(ComponentPose=@n17, Alpha=1, GravityOverride:Vector="0,0,-980", ExternalForce="0,0,0") -> (Pose) {
    angularBiasOverride=0.2
    angularDampingOverride=0.999
    angularSpringConstant=14000
    bAngularSpring=true
    bOverrideAngularBias=true
    bOverrideAngularDamping=true
    bUseGravityOverride=true
    gravityScale=0.35
    alphaBoolBlend {blendOption=Linear}
    boundBone {boneName=dyn_hair_t}
    componentAppliedLinearAccClamp=<200000 200000 200000>
    componentLinearAccScale=<1 1 1>
    componentLinearVelScale=<1 1 1>
    gravityOverride {z=-980}
    physicsBodyDefinitions [{
      boundBone {boneName=dyn_hair_t}
      boxExtents=<30 30 30>
      constraintSetup {angularLimitsMax=<7 20 7>, angularLimitsMin=<-7 -7.5 -7>, angularTarget=<1 0 0>}
      localJointOffset=<1.4 0 0>
     }]
    relativeSpaceBone {boneName=head}
   }
   n19 = Anim Dynamics - Bone: dyn_hair_kk!(ComponentPose=@n18, Alpha=1, GravityOverride:Vector="0,0,-980", ExternalForce="0,0,0") -> (Pose) {
    angularBiasOverride=0.2
    angularDampingOverride=0.999
    angularSpringConstant=14000
    bAngularSpring=true
    bOverrideAngularBias=true
    bOverrideAngularDamping=true
    bUseGravityOverride=true
    gravityScale=0.35
    alphaBoolBlend {blendOption=Linear}
    boundBone {boneName=dyn_hair_kk}
    componentAppliedLinearAccClamp=<200000 200000 200000>
    componentLinearAccScale=<1 1 1>
    componentLinearVelScale=<1 1 1>
    gravityOverride {z=-980}
    physicsBodyDefinitions [{
      boundBone {boneName=dyn_hair_kk}
      boxExtents=<30 30 30>
      constraintSetup {angularLimitsMax=<7 20 7>, angularLimitsMin=<-7 -7.5 -7>, angularTarget=<1 0 0>}
      localJointOffset=<1.4 0 0>
     }]
    relativeSpaceBone {boneName=head}
   }
   n20 = Anim Dynamics - Bone: dyn_hair_bb!(ComponentPose=@n19, Alpha=1, GravityOverride:Vector="0,0,-980", ExternalForce="0,0,0") -> (Pose) {
    angularBiasOverride=0.2
    angularDampingOverride=0.999
    angularSpringConstant=14000
    bAngularSpring=true
    bOverrideAngularBias=true
    bOverrideAngularDamping=true
    bUseGravityOverride=true
    gravityScale=0.35
    alphaBoolBlend {blendOption=Linear}
    boundBone {boneName=dyn_hair_bb}
    componentAppliedLinearAccClamp=<200000 200000 200000>
    componentLinearAccScale=<1 1 1>
    componentLinearVelScale=<1 1 1>
    gravityOverride {z=-980}
    physicsBodyDefinitions [{
      boundBone {boneName=dyn_hair_bb}
      boxExtents=<30 30 30>
      constraintSetup {angularLimitsMax=<7 20 7>, angularLimitsMin=<-7 -7.5 -7>, angularTarget=<1 0 0>}
      localJointOffset=<1.4 0 0>
     }]
    relativeSpaceBone {boneName=head}
   }
   n21 = Anim Dynamics - Bone: dyn_hair_k!(ComponentPose=@n20, Alpha=1, GravityOverride:Vector="0,0,-980", ExternalForce="0,0,0") -> (Pose) {
    angularBiasOverride=0.2
    angularDampingOverride=0.9999
    angularSpringConstant=14000
    bAngularSpring=true
    bOverrideAngularBias=true
    bOverrideAngularDamping=true
    bUseGravityOverride=true
    gravityScale=0.35
    alphaBoolBlend {blendOption=Linear}
    boundBone {boneName=dyn_hair_k}
    componentAppliedLinearAccClamp=<200000 200000 200000>
    componentLinearAccScale=<1 1 1>
    componentLinearVelScale=<1 1 1>
    gravityOverride {z=-980}
    physicsBodyDefinitions [{
      boundBone {boneName=dyn_hair_k}
      boxExtents=<40 50 50>
      constraintSetup {angularLimitsMax=<7.5 7.5 7.5>, angularLimitsMin=<-7.5 -7.5 -7.5>, angularTarget=<1 0 0>}
      localJointOffset=<1.1 0 0>
     }]
    relativeSpaceBone {boneName=head}
   }
   n22 = Anim Dynamics - Bone: dyn_hair_l!(ComponentPose=@n21, Alpha=1, GravityOverride:Vector="0,0,-980", ExternalForce="0,0,0") -> (Pose) {
    angularBiasOverride=0.2
    angularDampingOverride=0.9999
    angularSpringConstant=14000
    bAngularSpring=true
    bOverrideAngularBias=true
    bOverrideAngularDamping=true
    bUseGravityOverride=true
    gravityScale=0.35
    alphaBoolBlend {blendOption=Linear}
    boundBone {boneName=dyn_hair_l}
    componentAppliedLinearAccClamp=<200000 200000 200000>
    componentLinearAccScale=<1 1 1>
    componentLinearVelScale=<1 1 1>
    gravityOverride {z=-980}
    physicsBodyDefinitions [{
      boundBone {boneName=dyn_hair_l}
      boxExtents=<40 50 50>
      constraintSetup {angularLimitsMax=<7.5 7.5 7.5>, angularLimitsMin=<-7.5 -7.5 -7.5>, angularTarget=<1 0 0>}
      localJointOffset=<1.1 0 0>
     }]
    relativeSpaceBone {boneName=head}
   }
   n23 = Anim Dynamics - Bone: dyn_hair_o!(ComponentPose=@n22, Alpha=1, GravityOverride:Vector="0,0,-980", ExternalForce="0,0,0") -> (Pose) {
    angularBiasOverride=0.2
    angularDampingOverride=0.9999
    angularSpringConstant=14000
    bAngularSpring=true
    bOverrideAngularBias=true
    bOverrideAngularDamping=true
    bUseGravityOverride=true
    gravityScale=0.35
    alphaBoolBlend {blendOption=Linear}
    boundBone {boneName=dyn_hair_o}
    componentAppliedLinearAccClamp=<200000 200000 200000>
    componentLinearAccScale=<1 1 1>
    componentLinearVelScale=<1 1 1>
    gravityOverride {z=-980}
    physicsBodyDefinitions [{
      boundBone {boneName=dyn_hair_o}
      boxExtents=<40 50 50>
      constraintSetup {angularLimitsMax=<7.5 7.5 7.5>, angularLimitsMin=<-7.5 -7.5 -7.5>, angularTarget=<1 0 0>}
      localJointOffset=<1.1 0 0>
     }]
    relativeSpaceBone {boneName=head}
   }
   n24 = Anim Dynamics - Bone: dyn_hair_m!(ComponentPose=@n23, Alpha=1, GravityOverride:Vector="0,0,-980", ExternalForce="0,0,0") -> (Pose) {
    angularBiasOverride=0.2
    angularDampingOverride=0.9999
    angularSpringConstant=14000
    bAngularSpring=true
    bOverrideAngularBias=true
    bOverrideAngularDamping=true
    bUseGravityOverride=true
    gravityScale=0.35
    alphaBoolBlend {blendOption=Linear}
    boundBone {boneName=dyn_hair_m}
    componentAppliedLinearAccClamp=<200000 200000 200000>
    componentLinearAccScale=<1 1 1>
    componentLinearVelScale=<1 1 1>
    gravityOverride {z=-980}
    physicsBodyDefinitions [{
      boundBone {boneName=dyn_hair_m}
      boxExtents=<40 50 50>
      constraintSetup {angularLimitsMax=<7.5 7.5 7.5>, angularLimitsMin=<-7.5 -7.5 -7.5>, angularTarget=<1 0 0>}
      localJointOffset=<1.1 0 0>
     }]
    relativeSpaceBone {boneName=head}
   }
   n25 = Anim Dynamics - Bone: dyn_hair_x!(ComponentPose=@n24, Alpha=1, GravityOverride:Vector="0,0,-980", ExternalForce="0,0,0") -> (Pose) {
    angularBiasOverride=0.2
    angularDampingOverride=0.9999
    angularSpringConstant=14000
    bAngularSpring=true
    bOverrideAngularBias=true
    bOverrideAngularDamping=true
    bUseGravityOverride=true
    gravityScale=0.35
    alphaBoolBlend {blendOption=Linear}
    boundBone {boneName=dyn_hair_x}
    componentAppliedLinearAccClamp=<200000 200000 200000>
    componentLinearAccScale=<1 1 1>
    componentLinearVelScale=<1 1 1>
    gravityOverride {z=-980}
    physicsBodyDefinitions [{
      boundBone {boneName=dyn_hair_x}
      boxExtents=<40 50 50>
      constraintSetup {angularLimitsMax=<7.5 7.5 7.5>, angularLimitsMin=<-7.5 -7.5 -7.5>, angularTarget=<1 0 0>}
      localJointOffset=<1.1 0 0>
     }]
    relativeSpaceBone {boneName=head}
   }
   n26 = Anim Dynamics - Bone: dyn_hair_e!(ComponentPose=@n25, Alpha=1, GravityOverride:Vector="0,0,-980", ExternalForce="0,0,0") -> (Pose) {
    angularBiasOverride=0.2
    angularDampingOverride=0.9999
    angularSpringConstant=14000
    bAngularSpring=true
    bOverrideAngularBias=true
    bOverrideAngularDamping=true
    bUseGravityOverride=true
    gravityScale=0.35
    alphaBoolBlend {blendOption=Linear}
    boundBone {boneName=dyn_hair_e}
    componentAppliedLinearAccClamp=<200000 200000 200000>
    componentLinearAccScale=<1 1 1>
    componentLinearVelScale=<1 1 1>
    gravityOverride {z=-980}
    physicsBodyDefinitions [{
      boundBone {boneName=dyn_hair_e}
      boxExtents=<40 50 50>
      constraintSetup {angularLimitsMax=<7.5 7.5 7.5>, angularLimitsMin=<-7.5 -7.5 -7.5>, angularTarget=<1 0 0>}
      localJointOffset=<1.1 0 0>
     }]
    relativeSpaceBone {boneName=head}
   }
   n27 = Anim Dynamics - Bone: dyn_hair_u!(ComponentPose=@n26, Alpha=1, GravityOverride:Vector="0,0,-980", ExternalForce="0,0,0") -> (Pose) {
    angularBiasOverride=0.2
    angularDampingOverride=0.9999
    angularSpringConstant=14000
    bAngularSpring=true
    bOverrideAngularBias=true
    bOverrideAngularDamping=true
    bUseGravityOverride=true
    gravityScale=0.35
    alphaBoolBlend {blendOption=Linear}
    boundBone {boneName=dyn_hair_u}
    componentAppliedLinearAccClamp=<200000 200000 200000>
    componentLinearAccScale=<1 1 1>
    componentLinearVelScale=<1 1 1>
    gravityOverride {z=-980}
    physicsBodyDefinitions [{
      boundBone {boneName=dyn_hair_u}
      boxExtents=<40 50 50>
      constraintSetup {angularLimitsMax=<7.5 7.5 7.5>, angularLimitsMin=<-7.5 -7.5 -7.5>, angularTarget=<1 0 0>}
      localJointOffset=<1.1 0 0>
     }]
    relativeSpaceBone {boneName=head}
   }
   n28 = Anim Dynamics - Bone: dyn_hair_z!(ComponentPose=@n27, Alpha=1, GravityOverride:Vector="0,0,-980", ExternalForce="0,0,0") -> (Pose) {
    angularBiasOverride=0.2
    angularDampingOverride=0.9999
    angularSpringConstant=14000
    bAngularSpring=true
    bOverrideAngularBias=true
    bOverrideAngularDamping=true
    bUseGravityOverride=true
    gravityScale=0.35
    alphaBoolBlend {blendOption=Linear}
    boundBone {boneName=dyn_hair_z}
    componentAppliedLinearAccClamp=<200000 200000 200000>
    componentLinearAccScale=<1 1 1>
    componentLinearVelScale=<1 1 1>
    gravityOverride {z=-980}
    physicsBodyDefinitions [{
      boundBone {boneName=dyn_hair_z}
      boxExtents=<40 50 50>
      constraintSetup {angularLimitsMax=<7.5 7.5 7.5>, angularLimitsMin=<-7.5 -7.5 -7.5>, angularTarget=<1 0 0>}
      localJointOffset=<1.1 0 0>
     }]
    relativeSpaceBone {boneName=head}
   }
   n29 = Anim Dynamics - Bone: dyn_hair_zz!(ComponentPose=@n28, Alpha=1, GravityOverride:Vector="0,0,-980", ExternalForce="0,0,0") -> (Pose) {
    angularBiasOverride=0.2
    angularDampingOverride=0.9999
    angularSpringConstant=14000
    bAngularSpring=true
    bOverrideAngularBias=true
    bOverrideAngularDamping=true
    bUseGravityOverride=true
    gravityScale=0.35
    alphaBoolBlend {blendOption=Linear}
    boundBone {boneName=dyn_hair_zz}
    componentAppliedLinearAccClamp=<200000 200000 200000>
    componentLinearAccScale=<1 1 1>
    componentLinearVelScale=<1 1 1>
    gravityOverride {z=-980}
    physicsBodyDefinitions [{
      boundBone {boneName=dyn_hair_zz}
      boxExtents=<40 50 50>
      constraintSetup {angularLimitsMax=<7.5 7.5 7.5>, angularLimitsMin=<-7.5 -7.5 -7.5>, angularTarget=<1 0 0>}
      localJointOffset=<1.1 0 0>
     }]
    relativeSpaceBone {boneName=head}
   }
   n30 = Anim Dynamics - Bone: dyn_hair_xx!(ComponentPose=@n29, Alpha=1, GravityOverride:Vector="0,0,-980", ExternalForce="0,0,0") -> (Pose) {
    angularBiasOverride=0.2
    angularDampingOverride=0.9999
    angularSpringConstant=14000
    bAngularSpring=true
    bOverrideAngularBias=true
    bOverrideAngularDamping=true
    bUseGravityOverride=true
    gravityScale=0.35
    alphaBoolBlend {blendOption=Linear}
    boundBone {boneName=dyn_hair_xx}
    componentAppliedLinearAccClamp=<200000 200000 200000>
    componentLinearAccScale=<1 1 1>
    componentLinearVelScale=<1 1 1>
    gravityOverride {z=-980}
    physicsBodyDefinitions [{
      boundBone {boneName=dyn_hair_xx}
      boxExtents=<40 50 50>
      constraintSetup {angularLimitsMax=<7.5 7.5 7.5>, angularLimitsMin=<-7.5 -7.5 -7.5>, angularTarget=<1 0 0>}
      localJointOffset=<1.1 0 0>
     }]
    relativeSpaceBone {boneName=head}
   }
   n31 = Anim Dynamics - Bone: dyn_hair_i!(ComponentPose=@n30, Alpha=1, GravityOverride:Vector="0,0,-980", ExternalForce="0,0,0") -> (Pose) {
    angularBiasOverride=0.2
    angularDampingOverride=0.9999
    angularSpringConstant=14000
    bAngularSpring=true
    bOverrideAngularBias=true
    bOverrideAngularDamping=true
    bUseGravityOverride=true
    gravityScale=0.35
    alphaBoolBlend {blendOption=Linear}
    boundBone {boneName=dyn_hair_i}
    componentAppliedLinearAccClamp=<200000 200000 200000>
    componentLinearAccScale=<1 1 1>
    componentLinearVelScale=<1 1 1>
    gravityOverride {z=-980}
    physicsBodyDefinitions [{
      boundBone {boneName=dyn_hair_i}
      boxExtents=<40 50 50>
      constraintSetup {angularLimitsMax=<7.5 7.5 7.5>, angularLimitsMin=<-7.5 -7.5 -7.5>, angularTarget=<1 0 0>}
      localJointOffset=<1.1 0 0>
     }]
    relativeSpaceBone {boneName=head}
   }
   n32 = Anim Dynamics - Bone: dyn_hair_uu!(ComponentPose=@n31, Alpha=1, GravityOverride:Vector="0,0,-980", ExternalForce="0,0,0") -> (Pose) {
    angularBiasOverride=0.2
    angularDampingOverride=0.9999
    angularSpringConstant=14000
    bAngularSpring=true
    bOverrideAngularBias=true
    bOverrideAngularDamping=true
    bUseGravityOverride=true
    gravityScale=0.35
    alphaBoolBlend {blendOption=Linear}
    boundBone {boneName=dyn_hair_uu}
    componentAppliedLinearAccClamp=<200000 200000 200000>
    componentLinearAccScale=<1 1 1>
    componentLinearVelScale=<1 1 1>
    gravityOverride {z=-980}
    physicsBodyDefinitions [{
      boundBone {boneName=dyn_hair_uu}
      boxExtents=<40 50 50>
      constraintSetup {angularLimitsMax=<7.5 7.5 7.5>, angularLimitsMin=<-7.5 -7.5 -7.5>, angularTarget=<1 0 0>}
      localJointOffset=<1.1 0 0>
     }]
    relativeSpaceBone {boneName=head}
   }
   n33 = Anim Dynamics - Bone: dyn_hair_j!(ComponentPose=@n32, Alpha=1, GravityOverride:Vector="0,0,-980", ExternalForce="0,0,0") -> (Pose) {
    angularBiasOverride=0.2
    angularDampingOverride=0.9999
    angularSpringConstant=14000
    bAngularSpring=true
    bOverrideAngularBias=true
    bOverrideAngularDamping=true
    bUseGravityOverride=true
    gravityScale=0.35
    alphaBoolBlend {blendOption=Linear}
    boundBone {boneName=dyn_hair_j}
    componentAppliedLinearAccClamp=<200000 200000 200000>
    componentLinearAccScale=<1 1 1>
    componentLinearVelScale=<1 1 1>
    gravityOverride {z=-980}
    physicsBodyDefinitions [{
      boundBone {boneName=dyn_hair_j}
      boxExtents=<40 50 50>
      constraintSetup {angularLimitsMax=<7.5 7.5 7.5>, angularLimitsMin=<-7.5 -7.5 -7.5>, angularTarget=<1 0 0>}
      localJointOffset=<1.1 0 0>
     }]
    relativeSpaceBone {boneName=head}
   }
   n34 = Anim Dynamics - Bone: dyn_hair_ww!(ComponentPose=@n33, Alpha=1, GravityOverride:Vector="0,0,-980", ExternalForce="0,0,0") -> (Pose) {
    angularBiasOverride=0.2
    angularDampingOverride=0.9999
    angularSpringConstant=14000
    bAngularSpring=true
    bOverrideAngularBias=true
    bOverrideAngularDamping=true
    bUseGravityOverride=true
    gravityScale=0.35
    alphaBoolBlend {blendOption=Linear}
    boundBone {boneName=dyn_hair_ww}
    componentAppliedLinearAccClamp=<200000 200000 200000>
    componentLinearAccScale=<1 1 1>
    componentLinearVelScale=<1 1 1>
    gravityOverride {z=-980}
    physicsBodyDefinitions [{
      boundBone {boneName=dyn_hair_ww}
      boxExtents=<40 50 50>
      constraintSetup {angularLimitsMax=<7.5 7.5 7.5>, angularLimitsMin=<-7.5 -7.5 -7.5>, angularTarget=<1 0 0>}
      localJointOffset=<1.1 0 0>
     }]
    relativeSpaceBone {boneName=head}
   }
   n35 = Anim Dynamics - Bone: dyn_hair_aa!(ComponentPose=@n34, Alpha=1, GravityOverride:Vector="0,0,-980", ExternalForce="0,0,0") -> (Pose) {
    angularBiasOverride=0.2
    angularDampingOverride=0.9999
    angularSpringConstant=14000
    bAngularSpring=true
    bOverrideAngularBias=true
    bOverrideAngularDamping=true
    bUseGravityOverride=true
    gravityScale=0.35
    alphaBoolBlend {blendOption=Linear}
    boundBone {boneName=dyn_hair_aa}
    componentAppliedLinearAccClamp=<200000 200000 200000>
    componentLinearAccScale=<1 1 1>
    componentLinearVelScale=<1 1 1>
    gravityOverride {z=-980}
    physicsBodyDefinitions [{
      boundBone {boneName=dyn_hair_aa}
      boxExtents=<40 50 50>
      constraintSetup {angularLimitsMax=<7.5 7.5 7.5>, angularLimitsMin=<-7.5 -7.5 -7.5>, angularTarget=<1 0 0>}
      localJointOffset=<1.1 0 0>
     }]
    relativeSpaceBone {boneName=head}
   }
   n36 = Anim Dynamics - Bone: dyn_hair_dd!(ComponentPose=@n35, Alpha=1, GravityOverride:Vector="0,0,-980", ExternalForce="0,0,0") -> (Pose) {
    angularBiasOverride=0.2
    angularDampingOverride=0.9999
    angularSpringConstant=14000
    bAngularSpring=true
    bOverrideAngularBias=true
    bOverrideAngularDamping=true
    bUseGravityOverride=true
    gravityScale=0.35
    alphaBoolBlend {blendOption=Linear}
    boundBone {boneName=dyn_hair_dd}
    componentAppliedLinearAccClamp=<200000 200000 200000>
    componentLinearAccScale=<1 1 1>
    componentLinearVelScale=<1 1 1>
    gravityOverride {z=-980}
    physicsBodyDefinitions [{
      boundBone {boneName=dyn_hair_dd}
      boxExtents=<40 50 50>
      constraintSetup {angularLimitsMax=<7.5 7.5 7.5>, angularLimitsMin=<-7.5 -7.5 -7.5>, angularTarget=<1 0 0>}
      localJointOffset=<1.1 0 0>
     }]
    relativeSpaceBone {boneName=head}
   }
   Output Pose(Result=Component To Local(ComponentPose=@n36)) {layerGroup=None}
  ]}
 EventGraph {nodes [
   Event BlueprintUpdateAnimation()
   > n1 = Branch(Condition=Get bIsSkydiveFloating())
   > Set buckle_Gravity Override(buckle_Gravity Override:Vector="0,0,5000") -> (Output_Get:Vector)
   Set buckle_Gravity Override(execute=@n1.else, buckle_Gravity Override="0,0,-980")
  ]}
]
variables [
 buckle_Gravity Override {type=Vector}
 hair speed scale {type=number}
 skydiving cr {type=boolean}
]
